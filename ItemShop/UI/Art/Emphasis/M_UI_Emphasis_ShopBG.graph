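# UI material for the Emphasis item shop background, built from procedural color layers.
# The warm composite blends brown and dark tones through masks, adding radial highlights and a vertical orange falloff.
# Named reroutes declare concrete and grunge masks, a randomized particle-grid pattern, and gold embers from a confetti function with L1 parameters.
# A separate green branch blends a panning, flow-map-distorted flag texture under a vertical gradient.
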Material /ItemShop/UI/Art/Emphasis/M_UI_Emphasis_ShopBG
graph {
 properties {bUsedWithStaticMesh=false, materialDomain=MD_UI, pixelDepthOffsetMode=PDOM_Legacy, usedShadingModels=Default Lit}
 nodes [
  n1 = Constant3Vector() {constant=<rgba 0.0414 0.0833 0.0383 1>}
  n2 = TextureObject() {samplerType=SAMPLERTYPE_Alpha, texture=/OfferCatalog/Art/FNCS22/T_UI_FlagTexture}
  n3 = Multiply(A=MaterialFunctionCall() {materialFunction=&MF_UI_UVScaling}.Scaled To Fill)
  n4 = Constant() {r=0.5}
  n5 = Constant() {r=-0.9187}
  n6 = Multiply(A=Panner(Coordinate=MaterialFunctionCall(UVs=@n3, Rotation Center=@n4, "Rotation Angle (0-1)"=@n5) {materialFunction=&CustomRotator}.Rotated Values) {speedX=0.15, speedY=0.05}) {constB=0.4}
  n7 = TextureObject() {samplerType=SAMPLERTYPE_LinearColor, texture=/Game/UI/Frontend/ItemShop/Art/WaterBackground/T_UI_WaterFlowMap}
  n8 = Constant() {r=0.3}
  n9 = Constant() {r=0.5}
  n10 = TextureCoordinate()
  n11 = Multiply(A=Subtract(A=ComponentMask(None=@n10) {g=true}) {constB=0.55}) {constB=1.25}
  LinearInterpolate(
   A=LinearInterpolate(
    A=Constant3Vector() {constant=<rgba 0.075 0.1458 0.0787 1>}
    B=@n1
    Alpha=MaterialFunctionCall!(Texture 2D=@n2, Texture UVs=@n6, FlowMap Texture 2D=@n7, FlowMap UVs, FlowIntensity=@n8, FlowSpeed=@n9) {materialFunction=/FortUILibrary/MaterialFunctions/Effects/MF_UI_FlowMap}.Result
   )
   B=Multiply(A=@n11, B=@n1)
   Alpha=@n11
  )
  n12 = Constant3Vector() {constant=<rgba 0.1301 0.0529 0.0242 1>}
  n13 = Constant3Vector() {constant=<rgba 0.0365 0.0085 0.0035 1>}
  n14 = NamedRerouteUsage()
  n15 = LinearInterpolate(A=LinearInterpolate(A=LinearInterpolate(A=@n12, B=@n13, Alpha=@n14), B=@n13, Alpha=NamedRerouteUsage()), B=@n13, Alpha=NamedRerouteUsage())
  n16 = Constant3Vector() {constant=<rgba 0.0052 0.0052 0.0052 1>}
  n17 = NamedRerouteUsage()
  n18 = NamedRerouteUsage()
  n19 = Constant2Vector() {g=-0.45, r=0.5}
  n20 = MaterialFunctionCall(
   UV=MaterialFunctionCall(UVs=Add(A=@n18, B=@n19), Texture Scale=Constant() {r=2}) {materialFunction=&ScaleUVsByCenter}.UVs
   GradientStart=Constant()
   GradientEnd=Constant() {r=1}
  ) {materialFunction=&MF_UI_RadialGradient}
  n21 = Constant3Vector() {constant=<rgba 0.3073 0.1129 0 1>}
  n22 = Constant3Vector() {constant=<rgba 0.2552 0.0936 0.0093 1>}
  n23 = Max(A=NamedRerouteUsage(), B=Max(A=NamedRerouteUsage(), B=NamedRerouteUsage()))
  n24 = LinearInterpolate(
   A=LinearInterpolate(A=LinearInterpolate(A=@n15, B=@n16, Alpha=@n17), B=@n13, Alpha=NamedRerouteUsage())
   B=Max(A=Multiply(A=@n20.Smooth, B=@n21), B=Multiply(A=@n22, B=@n23))
   Alpha=Multiply(A=Max(A=@n20.Smooth, B=@n23)) {constB=0.45}
  )
  n25 = NamedRerouteUsage()
  n26 = ComponentMask(None=TextureCoordinate()) {g=true}
  n27 = Multiply(
   A=Subtract(A=OneMinus(None=@n26), B=ScalarParameter() {defaultValue=-1.4447, parameterName=SubtractYellow})
   B=ScalarParameter() {defaultValue=0.6192, parameterName=FalloffYellow}
  )
  n28 = NamedRerouteUsage()
  n29 = MaterialFunctionCall(UV=Add(A=Multiply(A=@n28) {constB=0.5}, B=Constant2Vector() {g=0.2493, r=0.276}), GradientStart=Constant(), GradientEnd=Constant() {r=1}) {materialFunction=&MF_UI_RadialGradient}
  LinearInterpolate(
   A=LinearInterpolate(A=Constant3Vector() {constant=<rgba 0.4844 0.1289 0.002 0>}, B=Max(A=@n24, B=@n25), Alpha=Saturate(None=@n27))
   B=Multiply(A=@n29.Smooth, B=Constant3Vector() {constant=<rgba 0.3872 0.036 0.0039 0>})
   Alpha=Multiply(A=@n29.Smooth) {constB=0.25}
  )
  n30 = NamedRerouteUsage()
  n31 = MaterialFunctionCall!(
   UV=Multiply(A=NamedRerouteUsage()) {constB=0.28}
   Global Offset=NamedRerouteUsage()
   Global Rotation=Constant() {r=0.35}
   Tiling=Constant() {r=6.5}
   Cell Offset=Constant() {r=0.4}
   Cell Scale=LinearInterpolate(Alpha=@n30) {constA=0.3, constB=0.65}
   Cell Rotation=Add(A=Constant() {r=1}, B=@n30)
   Random Seed
  ) {materialFunction=/FortUILibrary/MaterialFunctions/Basics/MF_UI_ParticleGrid}
  NamedRerouteDeclaration(None=@n31.Random Cell Values) {name=RCValues, nodeColor=<rgba 0 1 0.8125 1>}
  NamedRerouteDeclaration(None=@n31.Cells UV) {name=Random Pattern, nodeColor=<rgba 0 0.7656 1 1>}
  n32 = NamedRerouteUsage()
  n33 = AppendVector(A=ScalarParameter() {parameterName=L1-Patter-X-Offset}, B=ScalarParameter() {defaultValue=0.172, parameterName=L1-Patter-Y-Offset})
  n34 = ScalarParameter() {defaultValue=60, parameterName=L1-Pattern-X}
  n35 = ScalarParameter() {defaultValue=2, parameterName=L1-Pattern-Y}
  n36 = ScalarParameter() {defaultValue=20, parameterName=L1-Confetti-Local-Tile-X}
  n37 = ScalarParameter() {defaultValue=350, parameterName=L1-Confetti-Local-Tile-Y}
  n38 = ScalarParameter() {defaultValue=0.75, parameterName=L1-SmoothStep-Max}
  n39 = ScalarParameter() {defaultValue=0.508, parameterName=L1-SmoothStep-Min}
  n40 = ComponentMask(None=TextureCoordinate()) {g=true}
  n41 = Subtract(A=@n40, B=ScalarParameter() {defaultValue=0.236, parameterName=L1-Gradient-Mask})
  n42 = ScalarParameter() {defaultValue=2, parameterName=L1-Gradient-Falloff}
  n43 = Multiply(
   A=MaterialFunctionCall!(
    "UV(V2)"=Add(A=@n32, B=@n33)
    Global-Angle=ScalarParameter() {defaultValue=-1, parameterName=L1-Global Angle}
    "Pattern-Tiles(V2)"=AppendVector(A=@n34, B=@n35)
    RandomSeed=ScalarParameter() {defaultValue=12345, parameterName=L1-RandomSeed}
    Confetti-Speed-TimeFactor=ScalarParameter() {defaultValue=0.05, parameterName=L1-Confetti-Speed-TimeFactor}
    Confetti-Rotation-TimeFactor=ScalarParameter() {parameterName=L1-Confetti-Rotation-TimeFactor}
    Confetti-Local-Rotation-Range=ScalarParameter() {defaultValue=0.0075, parameterName=L1-Confetti-Local-Rotation-Range}
    "Confetti-Local-Tiles(V2)"=AppendVector(A=@n36, B=@n37)
    "SmoothStepRanges(V2)"=AppendVector(A=@n38, B=@n39)
   ) {materialFunction=/Game/UI/Frontend/ItemShop/Art/Cactus/MF_UI_Confetti}.Confetti-Mask
   B=Saturate(None=Multiply(A=@n41, B=@n42))
  )
  n44 = ScalarParameter() {defaultValue=1, parameterName=L1-Intensity}
  NamedRerouteDeclaration(None=Multiply(A=Constant3Vector() {constant=<rgba 1 0.7108 0 1>}, B=Multiply(A=@n43, B=@n44))) {name=Embers, nodeColor=<rgba 1 0 0.3281 1>}
  n45 = NamedRerouteUsage()
  n46 = TextureSample(UVs=MaterialFunctionCall(UVs=Multiply(A=@n45) {constB=0.95}, "Rotation Angle (0-1)"=Constant() {r=0.35}) {materialFunction=&CustomRotator}.Rotated Values) {samplerType=SAMPLERTYPE_Normal, texture=/ItemShop/UI/Art/Jumpsuit/T_UI_Grunge1}
  NamedRerouteDeclaration(None=Add(A=Multiply(A=@n46.R) {constB=0.25}) {constB=0.5}) {name=Back Most Grunge, nodeColor=<rgba 1 0.1406 0 1>}
  n47 = NamedRerouteUsage()
  n48 = TextureSample(UVs=Add(A=Multiply(A=@n47) {constB=0.95}, B=Constant2Vector() {g=-0.15, r=-0.65})) {samplerType=SAMPLERTYPE_Normal, texture=&/T_UI_Concrete}
  NamedRerouteDeclaration(None=Add(A=Multiply(A=@n48.G) {constB=0.5}) {constB=0.5}) {name=Grunge 1, nodeColor=<rgba 0.0781 1 0 1>}
  n49 = NamedRerouteUsage()
  n50 = TextureSample(UVs=Add(A=Multiply(A=@n49) {constB=0.95}, B=Constant2Vector() {g=0.15, r=0.65})) {samplerType=SAMPLERTYPE_Normal, texture=&T_UI_Concrete}
  NamedRerouteDeclaration(None=Add(A=Multiply(A=@n50.G) {constB=0.5}) {constB=0.5}) {name=Grunge 2, nodeColor=<rgba 1 0 0.2578 1>}
  n51 = NamedRerouteUsage()
  n52 = TextureSample(UVs=Add(A=Multiply(A=@n51), B=Constant2Vector() {g=-0.5, r=-0.4})) {samplerType=SAMPLERTYPE_Normal, texture=&T_UI_Concrete}
  NamedRerouteDeclaration(None=Add(A=Multiply(A=@n52.G) {constB=0.35}) {constB=0.5}) {name=Grunge 3, nodeColor=<rgba 0.1719 0 1 1>}
  n53 = TextureSample(UVs=NamedRerouteUsage()) {samplerType=SAMPLERTYPE_Normal, texture=&T_UI_Concrete}
  NamedRerouteDeclaration(None=Add(A=Multiply(A=@n53.R) {constB=0.5}) {constB=0.5}) {name=Top Grunge, nodeColor=<rgba 0.3125 1 0 1>}
  NamedRerouteDeclaration(None=@n31.Random Rows) {name=Random Rows, nodeColor=<rgba 1 0.3984 0 1>}
  NamedRerouteDeclaration(None=@n31.Random Columns) {name=random columns, nodeColor=<rgba 0.1719 0 1 1>}
  n54 = MaterialFunctionCall() {materialFunction=&MF_UI_UVScaling}
  NamedRerouteDeclaration(None=@n54.Scaled to Fit) {name=Scaled To Fit UVs, nodeColor=<rgba 0.8047 0 1 1>}
  NamedRerouteDeclaration(None=@n54.Scaled To Fill) {name=Scaled To Fill UVs, nodeColor=<rgba 0 1 0.8594 1>}
  n55 = Add(A=NamedRerouteUsage(), B=Constant2Vector() {g=0.138, r=-0.298})
  n56 = Constant() {r=0.65}
  NamedRerouteDeclaration(None=MaterialFunctionCall(
    UV=MaterialFunctionCall(UVs=@n55, Texture Scale=@n56) {materialFunction=&ScaleUVsByCenter}.UVs
    GradientStart=Constant()
    GradientEnd=Constant() {r=1}
   ) {materialFunction=&MF_UI_RadialGradient}.Smooth) {name=Right Middle Highlight, nodeColor=<rgba 0 1 0.1094 1>}
  n57 = Add(A=NamedRerouteUsage(), B=Constant2Vector() {g=0.3, r=0.7})
  n58 = Constant() {r=0.764}
  NamedRerouteDeclaration(None=MaterialFunctionCall(
    UV=MaterialFunctionCall(UVs=@n57, Texture Scale=@n58) {materialFunction=&ScaleUVsByCenter}.UVs
    GradientStart=Constant()
    GradientEnd=Constant() {r=1}
   ) {materialFunction=&MF_UI_RadialGradient}.Smooth) {name=Left Corner Highlight, nodeColor=<rgba 0 1 0.5313 1>}
  n59 = Add(A=NamedRerouteUsage(), B=Constant2Vector() {g=-0.266, r=-0.7092})
  n60 = Constant() {r=0.35}
  NamedRerouteDeclaration(None=MaterialFunctionCall(
    UV=MaterialFunctionCall(UVs=@n59, Texture Scale=@n60) {materialFunction=&ScaleUVsByCenter}.UVs
    GradientStart=Constant()
    GradientEnd=Constant() {r=1}
   ) {materialFunction=&MF_UI_RadialGradient}.Smooth) {name=Lower Right Highlight, nodeColor=<rgba 0.6875 0 1 1>}
  OneMinus(None=@n40)
 ]
}
